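; ChallengeQuestRequirement_NoChallenge is a conversation requirement that passes only when the participant player does not already have a challenge.
; IsRequirementSatisfied gets the player from the conversation context and checks their ability system component for the has-challenge gameplay tag.
; It returns Passed when the tag is absent and FailedAndHidden when present, hiding that conversation option.
; An invalid player gets no explicit result.
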(fn IsRequirementSatisfied
  (params (Context :type Conversation%20Context%20Structure%20%28by%20ref%29))
  (bind _returnvalue (Class|FortConversationContextLibrary|GetParticipantActorPlayer Context))
  (Utilities|IsValid _returnvalue
    (:Is%20Valid
      (if (not (GameplayTags|HasMatchingGameplayTag (Ability|GetAbilitySystemComponent _returnvalue) (Variables|Default|GetT_HasChallenge)))
        (LocalVariables|SetSatisfied true)
        (else
          (LocalVariables|SetSatisfied false)))
      (return (select (LocalVariables|GetSatisfied) "Passed" "FailedAndHidden")))
    (:Is%20Not%20Valid
      _)))
