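; A conversation task node for challenge quests that gives the talking player one copy of a configured item.
; On execution it gets the player participant from the conversation context, casts it to the Fort inventory owner interface and adds the Item Definition, then defers to the parent node.
; It also extends the parent's static extra data with an "item" parameter holding the item's object path.
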(fn ExecuteTaskNode
  (params (Context :type Conversation%20Context%20Structure%20%28by%20ref%29))
  (bind _item_definition (Variables|Default|GetItem%20Definition))
  (bind _returnvalue (Class|FortConversationContextLibrary|GetParticipantActorPlayer Context))
  (Utilities|IsValid _returnvalue
    (:Is%20Valid
      (bind _asfort_inventory_owner_interface (Utilities|Casting|CastToFortInventoryOwnerInterface _returnvalue)
        (:then
          (Utilities|IsValid _item_definition
            (:Is%20Valid
              (InventoryManagement|AddItem _asfort_inventory_owner_interface _item_definition 1))
            (:Is%20Not%20Valid
              _))
          (bind _returnvalue_2 (|CallParentFunction Context))
          (return _returnvalue_2))
        (:CastFailed
          _)))
    (:Is%20Not%20Valid
      _)))

(fn GatherStaticExtraData
  (params (Context :type Conversation%20Context%20Structure%20%28by%20ref%29))
  (bind _item_definition (Variables|Default|GetItem%20Definition))
  (bind _inoutextradata (|CallParentFunction Context))
  (Utilities|IsValid _item_definition
    (:Is%20Valid
      (Utilities|Array|Add _inoutextradata (Utilities|Struct|MakeConversationNodeParameterPair "item" (Utilities|GetObjectPathString _item_definition))))
    (:Is%20Not%20Valid
      _))
  (return _inoutextradata))
