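; Conversation task node that strips challenge quest items from the participating player's inventory.
; On execution it gets the player from the conversation context, reads their items through the Fort inventory owner interface and loops over them.
; Each item matching the challenge quest gameplay tag variable is removed in full through the player controller.
; With a valid player it returns the parent task's result; an invalid player skips the parent call.
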(fn ExecuteTaskNode
  (params (Context :type Conversation%20Context%20Structure%20%28by%20ref%29))
  (bind _returnvalue (Class|FortConversationContextLibrary|GetParticipantActorPlayer Context))
  (Utilities|IsValid _returnvalue
    (:Is%20Valid
      (bind _asfort_inventory_owner_interface (Utilities|Casting|CastToFortInventoryOwnerInterface _returnvalue)
        (:then
          (bind _items (InventoryManagement|GetItemsfromInventoryOwner _asfort_inventory_owner_interface))
          (bind _array_element (Utilities|Array|ForEachLoopwithBreak _items)
            (:LoopBody
              (if (GameplayTags|HasMatchingGameplayTag _array_element (Variables|Default|GetT_ChallengeQuest))
                (Inventory|RemoveFortItemFromPlayer (Fortnite|GetFortPlayerControllerfromActor _returnvalue) _array_element -1 true)))
            (:Completed
              _)))
        (:CastFailed
          _))
      (bind _returnvalue_2 (|CallParentFunction Context))
      (return _returnvalue_2))
    (:Is%20Not%20Valid
      _)))
