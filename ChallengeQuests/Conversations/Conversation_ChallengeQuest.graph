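# NPC conversation database offering six challenge quests: LlavaGround, DmgOnPickup, CannotReload, CannotUseHealingItems, DmgOnStandingStill and NoShields.
# Each tagged entry leads to speech offering Cancel, its own Accept node, and one shared Remove node.
# Accept requires no challenge, is gated on not reaching safe zone phase 3 (staying visible on failure), and triggers item grant, feed message and audio.
# Remove requires an existing challenge.
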ConversationDatabase /ChallengeQuests/Conversations/Conversation_ChallengeQuest
entry_points [
 {tag=Conversation.Entry.ChallengeQuests.LlavaGround, destinations [#1]}
 {tag=Conversation.Entry.ChallengeQuests.DmgOnPickup, destinations [#2]}
 {tag=Conversation.Entry.ChallengeQuests.CannotReload, destinations [#3]}
 {tag=Conversation.Entry.ChallengeQuests.CannotUseHealingItems, destinations [#4]}
 {tag=Conversation.Entry.ChallengeQuests.DmgOnStandingStill, destinations [#5]}
 {tag=Conversation.Entry.ChallengeQuests.NoShields, destinations [#6]}
]
nodes [
 ConversationEntryPointNode_5 {
  id=#6
  class=ConversationEntryPointNode
  outputs [#7]
  properties {entryTag {tagName=Conversation.Entry.ChallengeQuests.NoShields}}
 }
 Speech_C_5 {
  id=#7
  class=Speech_C
  outputs [#8, #9, #10]
  properties {nodeName=""}
 }
 ChallengeQuest_TaskNode_RemoveItem_C_0 {
  id=#10
  class=ChallengeQuest_TaskNode_RemoveItem_C
  outputs []
  sub_nodes [
   ConversationChoiceNode_6 {
    class=ConversationChoiceNode
    properties {
     defaultChoiceDisplayText=Remove Challenge
     nodeName=Remove
     choiceTags {gameplayTags [{tagName=Conversation.Choice.CustomSortPriority.Medium}, {tagName=Athena.Quests.S30.Challenges.TandemService.RemoveChallenge}]}
    }
   }
   SideEffect_PlayServiceUsedEffects_C_1 {class=SideEffect_PlayServiceUsedEffects_C}
   ChallengeQuestRequirement_HasChallenge_C_1 {class=ChallengeQuestRequirement_HasChallenge_C}
  ]
 }
 ChallengeQuest_TaskNode_GiveItem_C_1 {
  id=#9
  class=ChallengeQuest_TaskNode_GiveItem_C
  outputs []
  properties {
   item Definition=/ChallengeQuests/Items/Augments/NoShields/PAID_ChallengeQuest_NoShields
   serviceBriefCollectionKeyOveride=service.givechallengequest.accept.noshields
   effectRecipientConfigs [{
     recipient=Player
     effects [{class=FortControllerEffect_TriggerCustomFeedMessage, properties {customFeedMessageTag {tagName=Tandem.Service.ChallengeQuest}}}]
    }]
   requirements [{
     cannotUseReasonParameter=XStormCirclesPassed
     failureNodeBehaviour=FailedButVisible
     participantId {tagName=Conversation.Participant.Player}
     requirement {
      class=FortControllerRequirement_WorldCondition
      properties {conditions {
        description="IF NOT [Has reached SafeZone phase 3]"
        editableConditions [{
          bInvert=true
          expressionDepth=0
          operator=And
          condition {
           _structType=/Script/FortniteAI.FortWorldConditionAthenaGamePhase
           bInvert=true
           nextExpressionDepth=0
           operator=Copy
           phaseStep=None
           safeZonePhase=3
          }
         }]
       }}
     }
    }]
  }
  sub_nodes [
   ConversationChoiceNode_16 {
    class=ConversationChoiceNode
    properties {
     defaultChoiceDisplayText=Accept Challenge
     nodeName=Accept
     choiceTags {gameplayTags [{tagName=Conversation.Choice.CustomSortPriority.High}, {tagName=Athena.Quests.S30.Challenges.TandemService.GiveChallenge}]}
    }
   }
   ChallengeQuestRequirement_NoChallenge_C_3 {class=ChallengeQuestRequirement_NoChallenge_C}
   SideEffect_PlayServiceUsedEffects_C_5 {class=SideEffect_PlayServiceUsedEffects_C}
   SideEffect_Server_ChallengeQuest_PlayBigAudio_C_5 {class=SideEffect_Server_ChallengeQuest_PlayBigAudio_C}
  ]
 }
 Continue_C_5 {
  id=#8
  class=Continue_C
  properties {softTaskIcon=&T_UI_NPC_Back}
  sub_nodes [ConversationChoiceNode_15 {
    class=ConversationChoiceNode
    properties {defaultChoiceDisplayText=CANCEL, choiceTags {gameplayTags [{tagName=Conversation.Choice.CustomSortPriority.Highest}]}}
   }]
 }
 ConversationEntryPointNode_4 {
  id=#5
  class=ConversationEntryPointNode
  outputs [#11]
  properties {entryTag {tagName=Conversation.Entry.ChallengeQuests.DmgOnStandingStill}}
 }
 Speech_C_3 {
  id=#11
  class=Speech_C
  outputs [#12, #13, #10]
  properties {nodeName=""}
 }
 ChallengeQuest_TaskNode_GiveItem_C_0 {
  id=#13
  class=ChallengeQuest_TaskNode_GiveItem_C
  outputs []
  properties {
   item Definition=/ChallengeQuests/Items/Augments/DmgOnStandingStill/PAID_ChallengeQuest_DmgOnStandingStill
   serviceBriefCollectionKeyOveride=service.givechallengequest.accept.dmgonstandingstill
   effectRecipientConfigs [{
     recipient=Player
     effects [{class=FortControllerEffect_TriggerCustomFeedMessage, properties {customFeedMessageTag {tagName=Tandem.Service.ChallengeQuest}}}]
    }]
   requirements [{
     cannotUseReasonParameter=XStormCirclesPassed
     failureNodeBehaviour=FailedButVisible
     participantId {tagName=Conversation.Participant.Player}
     requirement {
      class=FortControllerRequirement_WorldCondition
      properties {conditions {
        description="IF NOT [Has reached SafeZone phase 3]"
        editableConditions [{
          bInvert=true
          expressionDepth=0
          operator=And
          condition {
           _structType=&FortWorldConditionAthenaGamePhase
           bInvert=true
           nextExpressionDepth=0
           operator=Copy
           phaseStep=None
           safeZonePhase=3
          }
         }]
       }}
     }
    }]
  }
  sub_nodes [
   ConversationChoiceNode_18 {
    class=ConversationChoiceNode
    properties {
     defaultChoiceDisplayText=Accept Challenge
     nodeName=Accept
     choiceTags {gameplayTags [{tagName=Conversation.Choice.CustomSortPriority.High}, {tagName=Athena.Quests.S30.Challenges.TandemService.GiveChallenge}]}
    }
   }
   ChallengeQuestRequirement_NoChallenge_C_4 {class=ChallengeQuestRequirement_NoChallenge_C}
   SideEffect_PlayServiceUsedEffects_C_6 {class=SideEffect_PlayServiceUsedEffects_C}
   SideEffect_Server_ChallengeQuest_PlayBigAudio_C_4 {class=SideEffect_Server_ChallengeQuest_PlayBigAudio_C}
  ]
 }
 Continue_C_4 {
  id=#12
  class=Continue_C
  properties {softTaskIcon=&T_UI_NPC_Back}
  sub_nodes [ConversationChoiceNode_17 {
    class=ConversationChoiceNode
    properties {defaultChoiceDisplayText=CANCEL, choiceTags {gameplayTags [{tagName=Conversation.Choice.CustomSortPriority.Highest}]}}
   }]
 }
 ConversationEntryPointNode_3 {
  id=#4
  class=ConversationEntryPointNode
  outputs [#14]
  properties {entryTag {tagName=Conversation.Entry.ChallengeQuests.CannotUseHealingItems}}
 }
 Speech_C_4 {
  id=#14
  class=Speech_C
  outputs [#15, #10, #16]
  properties {nodeName=""}
 }
 ChallengeQuest_TaskNode_GiveItem_C_8 {
  id=#16
  class=ChallengeQuest_TaskNode_GiveItem_C
  outputs []
  properties {
   item Definition=/ChallengeQuests/Items/Augments/NoHealingItems/PAID_ChallengeQuest_NoHealingItems
   serviceBriefCollectionKeyOveride=service.givechallengequest.accept.cannotusehealingitems
   effectRecipientConfigs [{
     recipient=Player
     effects [{class=FortControllerEffect_TriggerCustomFeedMessage, properties {customFeedMessageTag {tagName=Tandem.Service.ChallengeQuest}}}]
    }]
   requirements [{
     cannotUseReasonParameter=XStormCirclesPassed
     failureNodeBehaviour=FailedButVisible
     participantId {tagName=Conversation.Participant.Player}
     requirement {
      class=FortControllerRequirement_WorldCondition
      properties {conditions {
        description="IF NOT [Has reached SafeZone phase 3]"
        editableConditions [{
          bInvert=true
          expressionDepth=0
          operator=And
          condition {
           _structType=&FortWorldConditionAthenaGamePhase
           bInvert=true
           nextExpressionDepth=0
           operator=Copy
           phaseStep=None
           safeZonePhase=3
          }
         }]
       }}
     }
    }]
  }
  sub_nodes [
   ConversationChoiceNode_9 {
    class=ConversationChoiceNode
    properties {
     defaultChoiceDisplayText=Accept Challenge
     nodeName=Accept
     choiceTags {gameplayTags [{tagName=Conversation.Choice.CustomSortPriority.High}, {tagName=Athena.Quests.S30.Challenges.TandemService.GiveChallenge}]}
    }
   }
   ChallengeQuestRequirement_NoChallenge_C_1 {class=ChallengeQuestRequirement_NoChallenge_C}
   SideEffect_PlayServiceUsedEffects_C_3 {class=SideEffect_PlayServiceUsedEffects_C}
   SideEffect_Server_ChallengeQuest_PlayBigAudio_C_3 {class=SideEffect_Server_ChallengeQuest_PlayBigAudio_C}
  ]
 }
 Continue_C_2 {
  id=#15
  class=Continue_C
  properties {softTaskIcon=&T_UI_NPC_Back}
  sub_nodes [ConversationChoiceNode_8 {
    class=ConversationChoiceNode
    properties {defaultChoiceDisplayText=CANCEL, choiceTags {gameplayTags [{tagName=Conversation.Choice.CustomSortPriority.Highest}]}}
   }]
 }
 ConversationEntryPointNode_2 {
  id=#3
  class=ConversationEntryPointNode
  outputs [#17]
  properties {entryTag {tagName=Conversation.Entry.ChallengeQuests.CannotReload}}
 }
 Speech_C_2 {
  id=#17
  class=Speech_C
  outputs [#18, #10, #19]
  properties {nodeName=""}
 }
 ChallengeQuest_TaskNode_GiveItem_C_6 {
  id=#19
  class=ChallengeQuest_TaskNode_GiveItem_C
  outputs []
  properties {
   item Definition=/ChallengeQuests/Items/Augments/NoReloading/PAID_ChallengeQuest_NoReloading
   serviceBriefCollectionKeyOveride=service.givechallengequest.accept.cannotreload
   effectRecipientConfigs [{
     recipient=Player
     effects [{class=FortControllerEffect_TriggerCustomFeedMessage, properties {customFeedMessageTag {tagName=Tandem.Service.ChallengeQuest}}}]
    }]
   requirements [{
     cannotUseReasonParameter=XStormCirclesPassed
     failureNodeBehaviour=FailedButVisible
     participantId {tagName=Conversation.Participant.Player}
     requirement {
      class=FortControllerRequirement_WorldCondition
      properties {conditions {
        description="IF NOT [Has reached SafeZone phase 3]"
        editableConditions [{
          bInvert=true
          expressionDepth=0
          operator=And
          condition {
           _structType=&FortWorldConditionAthenaGamePhase
           bInvert=true
           nextExpressionDepth=0
           operator=Copy
           phaseStep=None
           safeZonePhase=3
          }
         }]
       }}
     }
    }]
  }
  sub_nodes [
   ConversationChoiceNode_12 {
    class=ConversationChoiceNode
    properties {
     defaultChoiceDisplayText=Accept Challenge
     nodeName=Accept
     choiceTags {gameplayTags [{tagName=Conversation.Choice.CustomSortPriority.High}, {tagName=Athena.Quests.S30.Challenges.TandemService.GiveChallenge}]}
    }
   }
   ChallengeQuestRequirement_NoChallenge_C_7 {class=ChallengeQuestRequirement_NoChallenge_C}
   SideEffect_PlayServiceUsedEffects_C_7 {class=SideEffect_PlayServiceUsedEffects_C}
   SideEffect_Server_ChallengeQuest_PlayBigAudio_C_2 {class=SideEffect_Server_ChallengeQuest_PlayBigAudio_C}
  ]
 }
 Continue_C_1 {
  id=#18
  class=Continue_C
  properties {softTaskIcon=&T_UI_NPC_Back}
  sub_nodes [ConversationChoiceNode_2 {
    class=ConversationChoiceNode
    properties {defaultChoiceDisplayText=CANCEL, choiceTags {gameplayTags [{tagName=Conversation.Choice.CustomSortPriority.Highest}]}}
   }]
 }
 ConversationEntryPointNode_1 {
  id=#2
  class=ConversationEntryPointNode
  outputs [#20]
  properties {entryTag {tagName=Conversation.Entry.ChallengeQuests.DmgOnPickup}}
 }
 Speech_C_1 {
  id=#20
  class=Speech_C
  outputs [#21, #10, #22]
  properties {nodeName=""}
 }
 ChallengeQuest_TaskNode_GiveItem_C_5 {
  id=#22
  class=ChallengeQuest_TaskNode_GiveItem_C
  outputs []
  properties {
   item Definition=/ChallengeQuests/Items/Augments/DmgOnPickup/PAID_ChallengeQuest_DmgOnPickup
   serviceBriefCollectionKeyOveride=service.givechallengequest.accept.dmgonpickup
   effectRecipientConfigs [{
     recipient=Player
     effects [{class=FortControllerEffect_TriggerCustomFeedMessage, properties {customFeedMessageTag {tagName=Tandem.Service.ChallengeQuest}}}]
    }]
   requirements [{
     cannotUseReasonParameter=XStormCirclesPassed
     failureNodeBehaviour=FailedButVisible
     participantId {tagName=Conversation.Participant.Player}
     requirement {
      class=FortControllerRequirement_WorldCondition
      properties {conditions {
        description="IF NOT [Has reached SafeZone phase 3]"
        editableConditions [{
          bInvert=true
          expressionDepth=0
          operator=And
          condition {
           _structType=&FortWorldConditionAthenaGamePhase
           bInvert=true
           nextExpressionDepth=0
           operator=Copy
           phaseStep=None
           safeZonePhase=3
          }
         }]
       }}
     }
    }]
  }
  sub_nodes [
   ConversationChoiceNode_14 {
    class=ConversationChoiceNode
    properties {
     defaultChoiceDisplayText=Accept Challenge
     nodeName=Accept
     choiceTags {gameplayTags [
       {tagName=Conversation.Choice.CustomSortPriority.High}
       {tagName=Collections}
       {tagName=Athena.Quests.S30.Challenges.TandemService.GiveChallenge}
      ]}
    }
   }
   ChallengeQuestRequirement_NoChallenge_C_9 {class=ChallengeQuestRequirement_NoChallenge_C}
   SideEffect_PlayServiceUsedEffects_C_9 {class=SideEffect_PlayServiceUsedEffects_C}
   SideEffect_Server_ChallengeQuest_PlayBigAudio_C_1 {class=SideEffect_Server_ChallengeQuest_PlayBigAudio_C}
  ]
 }
 Continue_C_0 {
  id=#21
  class=Continue_C
  properties {softTaskIcon=&T_UI_NPC_Back}
  sub_nodes [ConversationChoiceNode_0 {
    class=ConversationChoiceNode
    properties {defaultChoiceDisplayText=CANCEL, choiceTags {gameplayTags [{tagName=Conversation.Choice.CustomSortPriority.Highest}]}}
   }]
 }
 ConversationEntryPointNode_0 {
  id=#1
  class=ConversationEntryPointNode
  outputs [#23]
  properties {entryTag {tagName=Conversation.Entry.ChallengeQuests.LlavaGround}}
 }
 Speech_C_0 {
  id=#23
  class=Speech_C
  outputs [#24, #10, #25]
  properties {nodeName=""}
 }
 ChallengeQuest_TaskNode_GiveItem_C_9 {
  id=#25
  class=ChallengeQuest_TaskNode_GiveItem_C
  outputs []
  properties {
   serviceBriefCollectionKeyOveride=service.givechallengequest.accept.groundisllava
   effectRecipientConfigs [{
     recipient=Player
     effects [{class=FortControllerEffect_TriggerCustomFeedMessage, properties {customFeedMessageTag {tagName=Tandem.Service.ChallengeQuest}}}]
    }]
   requirements [{
     cannotUseReasonParameter=XStormCirclesPassed
     failureNodeBehaviour=FailedButVisible
     participantId {tagName=Conversation.Participant.Player}
     requirement {
      class=FortControllerRequirement_WorldCondition
      properties {conditions {
        description="IF NOT [Has reached SafeZone phase 3]"
        editableConditions [{
          bInvert=true
          expressionDepth=0
          operator=And
          condition {
           _structType=&FortWorldConditionAthenaGamePhase
           bInvert=true
           nextExpressionDepth=0
           operator=Copy
           phaseStep=None
           safeZonePhase=3
          }
         }]
       }}
     }
    }]
  }
  sub_nodes [
   ConversationChoiceNode_13 {
    class=ConversationChoiceNode
    properties {
     defaultChoiceDisplayText=Accept Challenge
     nodeName=Accept
     choiceTags {gameplayTags [{tagName=Conversation.Choice.CustomSortPriority.High}, {tagName=Athena.Quests.S30.Challenges.TandemService.GiveChallenge}]}
    }
   }
   ChallengeQuestRequirement_NoChallenge_C_8 {class=ChallengeQuestRequirement_NoChallenge_C}
   SideEffect_PlayServiceUsedEffects_C_8 {class=SideEffect_PlayServiceUsedEffects_C}
   SideEffect_Server_ChallengeQuest_PlayBigAudio_C_0 {class=SideEffect_Server_ChallengeQuest_PlayBigAudio_C}
  ]
 }
 Continue_C_3 {
  id=#24
  class=Continue_C
  properties {softTaskIcon=&T_UI_NPC_Back}
  sub_nodes [ConversationChoiceNode_1 {
    class=ConversationChoiceNode
    properties {defaultChoiceDisplayText=CANCEL, choiceTags {gameplayTags [{tagName=Conversation.Choice.CustomSortPriority.Highest}]}}
   }]
 }
]
