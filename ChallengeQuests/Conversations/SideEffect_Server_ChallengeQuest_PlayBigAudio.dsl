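; Conversation side effect for challenge quest stations that makes the station prop play its big audio on the server.
; It gets the NPC participant actor from the conversation context and checks that it is valid.
; It then casts that actor to the challenge quest station prop parent and calls PlayBigAudio, doing nothing if either check fails.
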(fn ServerCauseSideEffect
  (params (Context :type Conversation%20Context%20Structure%20%28by%20ref%29))
  ;[0 0 0] Spawn Challenge Quest
  (bind _returnvalue (Class|FortConversationContextLibrary|GetParticipantActorNPC Context))
  (Utilities|IsValid _returnvalue
    (:Is%20Valid
      (bind _asbp_station_prop_challenge_quests_parent (Utilities|Casting|CastToBP_StationProp_ChallengeQuests_Parent _returnvalue)
        (:then
          (Class|BPStationPropChallengeQuestsParent|PlayBigAudio _asbp_station_prop_challenge_quests_parent))
        (:CastFailed
          _))
      (return))
    (:Is%20Not%20Valid
      _)))
